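# Translucent, unlit, two-sided master material for DimensionalFrag offense extents effects, tinted by particle color.
# Color blends EdgeColor to EmissiveColor by depth fade from intersecting geometry, scaled by near and far brightness lerped on camera distance.
# Opacity is a vertical gradient band concentrated at geometry intersections, with Alpha filling in the remaining surface.
# Particle alpha, a soft depth fade and near-camera fade further attenuate opacity.
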Material /Game/Effects/Fort_Effects/Materials/Pickup/DimensionalFrag/M_DF_Offense_Extents_Master
graph {
 properties {blendMode=BLEND_Translucent, pixelDepthOffsetMode=PDOM_Legacy, shadingModel=MSM_Unlit, twoSided=true, usedShadingModels=Unlit}
 nodes [
  MaterialFunctionCall() {materialFunction=&Color_Time_of_Day}
  n1 = ScalarParameter() {defaultValue=1, parameterName=NearColorScalar}
  n2 = ScalarParameter() {defaultValue=1, parameterName=FarColorScalar}
  n3 = MaterialFunctionCall(Distance=ScalarParameter() {defaultValue=1000, parameterName=LerpColorDistance}) {materialFunction=&Near_Camera_Fade}
  n4 = VectorParameter() {group=Color, parameterName=EdgeColor, defaultValue=<rgba 0.5 0.5 0.5 1>}
  n5 = VectorParameter() {group=Color, parameterName=EmissiveColor, defaultValue=<rgba 1 1 1 1>}
  n6 = DepthFade(FadeDistance=ScalarParameter() {defaultValue=25, parameterName=EdgeThickness}) {fadeDistanceDefault=75}
  n7 = ParticleColor()
  Multiply(A=Multiply(A=LinearInterpolate(A=@n1, B=@n2, Alpha=@n3.Result), B=LinearInterpolate(A=@n4.RGB, B=@n5.RGB, Alpha=@n6)), B=@n7.RGB)
  n8 = MaterialFunctionCall() {materialFunction=&LinearGradient}
  n9 = ScalarParameter() {parameterName=Alpha}
  n10 = LinearInterpolate(
   A=Multiply(
    A=OneMinus(None=@n8.VGradient)
    B=MaterialFunctionCall() {materialFunction=&LinearGradient}.VGradient
   )
   B=Constant()
   Alpha=DepthFade(Opacity=OneMinus(None=@n9), FadeDistance=ScalarParameter() {defaultValue=25, parameterName=EdgeThickness}) {fadeDistanceDefault=75}
  )
  n11 = ScalarParameter() {defaultValue=15, parameterName=DepthFade}
  Multiply(
   A=Multiply(A=@n7.A, B=DepthFade(Opacity=@n10, FadeDistance=@n11) {fadeDistanceDefault=15})
   B=MaterialFunctionCall() {materialFunction=&Near_Camera_Fade}.Result
  )
 ]
}
