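# Material function that drives a vertical bobbing world position offset for Dimensional Fragment effects, exposed to the material library.
# A Time node feeds a Sine with a fixed period, producing an oscillation.
# That wave multiplies the BounceHeightZ vector input, which defaults to a Z-only offset, and outputs the result as Bounce Height.
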MaterialFunction /Game/Effects/Fort_Effects/Materials/_Functions/Dimensional_Frag_Bounce
graph {
 properties {bExposeToLibrary=true, description=Function to control Height of offset on Dimensional Fragments, libraryCategoriesText [Misc,FX_Common]}
 nodes [
  n1 = Time()
  n2 = FunctionInput() {bUsePreviewValueAsDefault=true, description=WPO offset height, inputName=BounceHeightZ, previewValue=<0 0 3 1>}
  FunctionOutput(None=Multiply(A=Sine(None=@n1) {desc=0, period=3.5}, B=@n2)) {desc=How much WPO offset the Dimensional Fragments get, description=WPO output, outputName=Bounce Height}
 ]
}
