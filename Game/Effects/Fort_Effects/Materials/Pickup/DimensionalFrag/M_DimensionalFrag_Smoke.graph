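# Translucent, unlit, two-sided smoke material for the Dimensional Frag pickup effect, usable on mesh particles.
# An 8x8 SubUV smoke flipbook blends two color parameters, tinted by particle color and passed through shared emissive adjustments.
# Opacity combines flipbook and particle alpha with view-angle, near-camera and depth fades, tuned per particle through Dynamic Parameters.
# A Dimensional Frag bounce material function is also included.
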Material /Game/Effects/Fort_Effects/Materials/Pickup/DimensionalFrag/M_DimensionalFrag_Smoke
graph {
 properties {
  bGenerateSphericalParticleNormals=true
  bUsedWithMeshParticles=true
  blendMode=BLEND_Translucent
  pixelDepthOffsetMode=PDOM_Legacy
  shadingModel=MSM_Unlit
  translucencyDirectionalLightingIntensity=0.5
  translucentSelfShadowDensityScale=1
  translucentShadowDensityScale=0.25
  twoSided=true
  usedShadingModels=Unlit
 }
 nodes [
  n1 = DynamicParameter() {defaultValue=<rgba 0 0 0 1>, paramNames [FadeDistance,EdgeFade,NearFade,Param4]}
  n2 = MaterialFunctionCall("FadeSensitivity (optional)"=@n1.EdgeFade) {materialFunction=&FadeBasedOnViewAngle}
  n3 = TextureSampleParameterSubUV!(UVs, Apply View MipBias) {parameterName=SubUV_Diffuse, texture=/Game/Effects/Fort_Effects/BluePrints/BeeHive/T_SmokeFlip_03_8x8}
  n4 = ParticleColor()
  n5 = Multiply(A=@n3.A, B=@n4.A)
  DepthFade(
   Opacity=Multiply(
    A=Multiply(A=@n2.Result, B=@n5)
    B=MaterialFunctionCall(Distance=@n1.NearFade) {materialFunction=&Near_Camera_Fade}.Result
   )
   FadeDistance=@n1.FadeDistance
  ) {fadeDistanceDefault=32}
  n6 = VectorParameter() {parameterName=Color1, defaultValue=<rgba 1 1 1 1>}
  n7 = VectorParameter() {parameterName=Color2, defaultValue=<rgba 0.5 0.5 0.5 1>}
  MaterialFunctionCall(Emissive=Multiply(A=LinearInterpolate(A=@n6.RGB, B=@n7.RGB, Alpha=@n3.RGB), B=@n4.RGB)) {materialFunction=&MF_FX_EmissiveAdjustments}
  MaterialFunctionCall!(BounceHeightZ) {materialFunction=/Game/Effects/Fort_Effects/Materials/_Functions/Dimensional_Frag_Bounce}
 ]
}
